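# VOTestCue is a test-map Sound Cue for voice-over testing that has no nodes and a null root, so it produces no audio.
# It carries only routing settings: the non-critical enemy VO sound class, reduced volume and a very large max distance.
# Despite the VO naming, it uses emote-music concurrency and has a send to the music audio-analysis submix.
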SoundCue /Game/Athena/Sounds/TestBlueprints/TestMapSounds/VOTestCue
root=null
nodes []
properties {
 maxDistance=2097152
 soundClassObject=/Game/Sounds/FortSoundClassesAndModes/Fort_VO_Enemy_NonCritical
 volumeMultiplier=0.6667
 concurrencyOverrides {resolutionRule=PreventNew}
 concurrencySet [/Game/Sounds/Concurrency_Rules/Athena_EmoteMusic_Conc]
 soundSubmixSends [{minSendLevel=0, soundSubmix=&Fortnite_AudioAnalysis_Music}]
}
